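# Behavior tree for Phoebe AI that, going by its name, handles a bot that is down but not out (DBNO) inside a vehicle by making it exit.
# A single root sequence on the BB_Phoebe blackboard waits about a second with random deviation, then runs the LeaveVehicle task without waiting for the vehicle to stop.
# The tree has no decorators or services, so it does not check the DBNO or in-vehicle state itself.
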BehaviorTree /PhoebeCore/BehaviorTrees/BT_DBNOInVehicle
blackboard=&BB_Phoebe
root {
 name=Sequence
 class=BTComposite_Sequence
 children [
  Wait {class=BTTask_Wait, properties {randomDeviation {defaultValue=0.5}, waitTime {defaultValue=1}}}
  LeaveVehicle {
   class=FortAthenaBTTask_LeaveVehicle
   children []
   decorators []
   properties {bIgnoreRestartSelf=false, bWaitVehicleStop=false, nodeName=LeaveVehicle}
   services []
  }
 ]
 decorators []
 properties {children [
   {childTask=&self:BTTask_Wait_0}
   {childTask=&self:FortAthenaBTTask_LeaveVehicle_0}
  ]}
}
