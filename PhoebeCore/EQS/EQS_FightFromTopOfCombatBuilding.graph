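# EQS query that picks a position for an AI bot fighting from atop a combat building, favoring spots near its current target.
# A small grid is generated around the querier's building-grid-snapped floor center and projected vertically onto geometry by line trace.
# Two score-only inverse-linear tests favor points closer to the target in 3D and closer to the querier's height, keeping the bot on its current level.
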EnvQuery /PhoebeCore/EQS/EQS_FightFromTopOfCombatBuilding
options [{
  generator {
   name=FortEnvQueryGenerator_SimpleGrid_0
   class=FortEnvQueryGenerator_SimpleGrid
   properties {
    bAutoSortTests=true
    bCanRunAsync=false
    envQueryResultNormalizationOption=Default
    generateAround=/Game/Athena/AI/EQS/Contexts/EnvQueryContext_Querier_SnappedToBuildingGrid_FloorCenter.EnvQueryContext_Querier_SnappedToBuildingGrid_FloorCenter_C
    gridSize=240
    navDataOverrideContext=null
    optionName=""
    overridenAgentNameForNavmesh=None
    randomDisplacementRatio=0
    spaceBetween=230
    projectionData {
     bCanDisableTrace=true
     bCanProjectDown=true
     bCanTraceOnGeometry=true
     bCanTraceOnNavMesh=true
     bOnlyBlockingHits=true
     bTraceComplex=false
     extentX=0
     extentY=10
     extentZ=10
     navigationFilter=null
     postProjectionVerticalOffset=0
     projectDown=1024
     projectUp=1024
     serializedChannel=ECC_GameTraceChannel11
     traceChannel=TraceTypeQuery10
     traceMode=GeometryByChannel
     traceProfileName=None
     traceShape=Line
     versionNum=1
    }
   }
  }
  tests [
   EnvQueryTest_Distance_1 {
    class=EnvQueryTest_Distance
    properties {
     distanceTo=&FortAthenaAIBot_EQSQueryContext_CurrentTarget
     filterType=Range
     floatValueMin=0
     scoringEquation=InverseLinear
     scoringFactor=1
     testComment=""
     testMode=Distance3D
     testPurpose=Score
    }
   }
   EnvQueryTest_Distance_0 {
    class=EnvQueryTest_Distance
    properties {
     distanceTo=&EnvQueryContext_Querier
     filterType=Range
     floatValueMin=0
     scoringEquation=InverseLinear
     scoringFactor=1
     testComment=""
     testMode=DistanceZ
     testPurpose=Score
    }
   }
  ]
 }]
