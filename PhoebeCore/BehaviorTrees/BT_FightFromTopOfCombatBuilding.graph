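# Behavior tree for Phoebe bots fighting from the top of a combat building.
# Root services drive weapon selection, ranged and throwable attacks, aiming and target focus.
# After an initial EQS-picked firing position, a looped selector alternates stepping back with crouched moves to randomized top-25% firing spots.
# A reload-status decorator aborts that lower-priority cycle to reload in cover, with all moves sharing one blackboard location key.
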BehaviorTree /PhoebeCore/BehaviorTrees/BT_FightFromTopOfCombatBuilding
blackboard=&/BB_FightFromTopOfCombatBuilding
root {
 name=Sequence
 class=BTComposite_Sequence
 children [
  Run EQS Query {
   class=BTTask_RunEQSQuery
   properties {
    nodeName=Run EQS Query
    blackboardKey {selectedKeyName=CombatFromTopOfCombatBuilding_Location}
    eQSRequest {queryTemplate=/PhoebeCore/EQS/EQS_FightFromTopOfCombatBuilding, runMode=SingleResult}
   }
  }
  MoveDirectlyToward {
   class=BTTask_MoveDirectlyToward
   children []
   decorators []
   properties {
    bIgnoreRestartSelf=false
    nodeName=MoveDirectlyToward
    acceptableRadius {defaultValue=10, key=None}
    bAllowPartialPath {defaultValue=true, key=None}
    bAllowStrafe {defaultValue=true, key=None}
    bProjectGoalLocation {defaultValue=false, key=None}
    bReachTestIncludesAgentRadius {defaultValue=false, key=None}
    bReachTestIncludesGoalRadius {defaultValue=false, key=None}
    bRequireNavigableEndLocation {defaultValue=false, key=None}
    bStartFromPreviousPath {defaultValue=false, key=None}
    bTrackMovingGoal {defaultValue=false, key=None}
    blackboardKey {bNoneIsAllowedValue=false, selectedKeyName=CombatFromTopOfCombatBuilding_Location}
    filterClass {baseClass=&NavigationQueryFilter, defaultValue=null, key=None}
    observedBlackboardValueTolerance {defaultValue=4.75, key=None}
   }
   services []
  }
  Wait - Shoot {
   class=FortAthenaBTTask_BotWait
   children []
   decorators []
   properties {
    bIgnoreRestartSelf=false
    nodeName=Wait - Shoot
    randomDeviation {defaultValue=1, key=None}
    waitCompleteKeySelector {bNoneIsAllowedValue=false, selectedKeyName=SelfActor}
    waitTime {defaultValue=2, key=None}
   }
   services []
  }
  Selector {
   class=BTComposite_Selector
   children [
    Sequence {
     class=BTComposite_Sequence
     children [
      Run EQS Query - Cover to reload {
       class=BTTask_RunEQSQuery
       properties {
        nodeName=Run EQS Query - Cover to reload
        blackboardKey {selectedKeyName=CombatFromTopOfCombatBuilding_Location}
        eQSRequest {queryTemplate=/PhoebeCore/EQS/EQS_ReloadFromTopOfCombatBuilding, runMode=SingleResult}
       }
      }
      MoveDirectlyToward {
       class=BTTask_MoveDirectlyToward
       children []
       decorators []
       properties {
        bIgnoreRestartSelf=false
        nodeName=MoveDirectlyToward
        acceptableRadius {defaultValue=10, key=None}
        bAllowPartialPath {defaultValue=true, key=None}
        bAllowStrafe {defaultValue=true, key=None}
        bProjectGoalLocation {defaultValue=false, key=None}
        bReachTestIncludesAgentRadius {defaultValue=false, key=None}
        bReachTestIncludesGoalRadius {defaultValue=false, key=None}
        bRequireNavigableEndLocation {defaultValue=false, key=None}
        bStartFromPreviousPath {defaultValue=false, key=None}
        bTrackMovingGoal {defaultValue=false, key=None}
        blackboardKey {bNoneIsAllowedValue=false, selectedKeyName=CombatFromTopOfCombatBuilding_Location}
        filterClass {baseClass=&NavigationQueryFilter, defaultValue=null, key=None}
        observedBlackboardValueTolerance {defaultValue=4.75, key=None}
       }
       services []
      }
      Wait - Reload {
       class=FortAthenaBTTask_BotWait
       children []
       decorators []
       properties {
        bIgnoreRestartSelf=false
        nodeName=Wait - Reload
        randomDeviation {defaultValue=0.5, key=None}
        waitCompleteKeySelector {bNoneIsAllowedValue=false, selectedKeyName=SelfActor}
        waitTime {defaultValue=1, key=None}
       }
       services []
      }
     ]
     decorators [Reloading? {
       class=BTDecorator_Blackboard
       properties {
        arithmeticOperation=GreaterOrEqual
        flowAbortMode=LowerPriority
        intValue=2
        nodeName=Reloading?
        stringValue=ExecutionSuccess
        blackboardKey {selectedKeyName=AIEvaluator_WeaponReload_ExecutionStatus}
       }
      }]
     properties {children [
       {childTask=&self:BTTask_RunEQSQuery_9}
       {childTask=&self:BTTask_MoveDirectlyToward_4}
       {childTask=&self:FortAthenaBTTask_BotWait_11}
      ]}
    }
    Sequence {
     class=BTComposite_Sequence
     children [
      Sequence {
       class=BTComposite_Sequence
       children [
        Run EQS Query - Step Back {
         class=BTTask_RunEQSQuery
         properties {
          nodeName=Run EQS Query - Step Back
          blackboardKey {selectedKeyName=CombatFromTopOfCombatBuilding_Location}
          eQSRequest {queryTemplate=/PhoebeCore/EQS/EQS_StepBackFromTopOfCombatBuilding, runMode=RandomBest25Pct}
         }
        }
        MoveDirectlyToward To Step Back {
         class=BTTask_MoveDirectlyToward
         children []
         decorators []
         properties {
          bIgnoreRestartSelf=false
          nodeName=MoveDirectlyToward To Step Back
          acceptableRadius {defaultValue=10, key=None}
          bAllowPartialPath {defaultValue=true, key=None}
          bAllowStrafe {defaultValue=true, key=None}
          bProjectGoalLocation {defaultValue=false, key=None}
          bReachTestIncludesAgentRadius {defaultValue=false, key=None}
          bReachTestIncludesGoalRadius {defaultValue=false, key=None}
          bRequireNavigableEndLocation {defaultValue=false, key=None}
          bStartFromPreviousPath {defaultValue=false, key=None}
          bTrackMovingGoal {defaultValue=false, key=None}
          blackboardKey {bNoneIsAllowedValue=false, selectedKeyName=CombatFromTopOfCombatBuilding_Location}
          filterClass {baseClass=&NavigationQueryFilter, defaultValue=null, key=None}
          observedBlackboardValueTolerance {defaultValue=4.75, key=None}
         }
         services []
        }
        Wait - Step Back {
         class=FortAthenaBTTask_BotWait
         children []
         decorators []
         properties {
          bIgnoreRestartSelf=false
          nodeName=Wait - Step Back
          randomDeviation {defaultValue=0, key=None}
          waitCompleteKeySelector {bNoneIsAllowedValue=false, selectedKeyName=SelfActor}
          waitTime {defaultValue=0.2, key=None}
         }
         services []
        }
       ]
       decorators []
       properties {children [
         {childTask=&self:BTTask_RunEQSQuery_7}
         {childTask=&self:BTTask_MoveDirectlyToward_2}
         {childTask=&self:FortAthenaBTTask_BotWait_10}
        ]}
      }
      Sequence {
       class=BTComposite_Sequence
       children [
        Run EQS Query - Shoot {
         class=BTTask_RunEQSQuery
         properties {
          nodeName=Run EQS Query - Shoot
          blackboardKey {selectedKeyName=CombatFromTopOfCombatBuilding_Location}
          eQSRequest {queryTemplate=&EQS_FightFromTopOfCombatBuilding, runMode=RandomBest25Pct}
         }
        }
        MoveDirectlyToward To Shoot {
         class=BTTask_MoveDirectlyToward
         children []
         decorators []
         properties {
          bIgnoreRestartSelf=false
          nodeName=MoveDirectlyToward To Shoot
          acceptableRadius {defaultValue=10, key=None}
          bAllowPartialPath {defaultValue=true, key=None}
          bAllowStrafe {defaultValue=true, key=None}
          bProjectGoalLocation {defaultValue=false, key=None}
          bReachTestIncludesAgentRadius {defaultValue=false, key=None}
          bReachTestIncludesGoalRadius {defaultValue=false, key=None}
          bRequireNavigableEndLocation {defaultValue=false, key=None}
          bStartFromPreviousPath {defaultValue=false, key=None}
          bTrackMovingGoal {defaultValue=false, key=None}
          blackboardKey {bNoneIsAllowedValue=false, selectedKeyName=CombatFromTopOfCombatBuilding_Location}
          filterClass {baseClass=&NavigationQueryFilter, defaultValue=null, key=None}
          observedBlackboardValueTolerance {defaultValue=4.75, key=None}
         }
         services []
        }
        Wait - Shoot {
         class=FortAthenaBTTask_BotWait
         children []
         decorators []
         properties {
          bIgnoreRestartSelf=false
          nodeName=Wait - Shoot
          randomDeviation {defaultValue=1, key=None}
          waitCompleteKeySelector {bNoneIsAllowedValue=false, selectedKeyName=SelfActor}
          waitTime {defaultValue=2, key=None}
         }
         services []
        }
       ]
       decorators []
       properties {
        children [
         {childTask=&self:BTTask_RunEQSQuery_8}
         {childTask=&self:BTTask_MoveDirectlyToward_3}
         {childTask=&self:FortAthenaBTTask_BotWait_15}
        ]
        services [&self:FortBTService_Crouch_1]
       }
       services [FortBTService_Crouch_1 {
         class=FortBTService_Crouch
         properties {
          bCallTickOnSearchStart=false
          bRestartTimerOnEachActivation=false
          interval=0.5
          nodeName=""
          randomDeviation=0.1
          enterShouldCrouch {defaultValue=true, key=None}
          exitShouldCrouch {defaultValue=false, key=None}
         }
        }]
      }
     ]
     decorators []
     properties {children [
       {childComposite=&self:BTComposite_Sequence_9, childTask=null}
       {childComposite=&self:BTComposite_Sequence_10, childTask=null}
      ]}
    }
   ]
   decorators [Loop {class=BTDecorator_Loop}]
   properties {
    children [
     {
      childComposite=&self:BTComposite_Sequence_7
      childTask=null
      decorators [&self:BTDecorator_Blackboard_1]
     }
     {childComposite=&self:BTComposite_Sequence_6, childTask=null, decorators []}
    ]
    services []
   }
   services []
  }
 ]
 decorators []
 properties {
  children [
   {childTask=&self:BTTask_RunEQSQuery_0}
   {childTask=&self:BTTask_MoveDirectlyToward_1}
   {childTask=&self:FortAthenaBTTask_BotWait_12}
   {
    childComposite=&self:BTComposite_Selector_2
    childTask=null
    decorators [&self:BTDecorator_Loop_5]
   }
  ]
  services [
   &self:FortAthenaBTService_SetInTacticalAction_1
   &self:FortAthenaBTService_AIEvaluator_29
   &self:FortAthenaBTService_AIEvaluator_30
   &self:FortAthenaBTService_AIEvaluator_31
   &self:FortAthenaBTService_AIEvaluator_32
   &self:FortAthenaBTService_ManageWeapon_7
   &self:FortAthenaBTService_AIEvaluator_28
  ]
 }
 services [
  FortAthenaBTService_SetInTacticalAction_1 {
   class=FortAthenaBTService_SetInTacticalAction
   properties {shouldSetInTacticalAction {key=None}, tacticalActionIdentifierTag {defaultValue {_structType=&GameplayTag}}}
  }
  Evaluating...Select Weapon {
   class=FortAthenaBTService_AIEvaluator
   properties {
    interval=0.5
    nodeName=Evaluating...Select Weapon
    randomDeviation=0.05
    aIEvaluatorClass {defaultValue=&FortAthenaAIBotEvaluator_WeaponSelection}
   }
  }
  Evaluating...Throwable Attack {
   class=FortAthenaBTService_AIEvaluator
   properties {
    interval=0.2
    nodeName=Evaluating...Throwable Attack
    randomDeviation=0.05
    aIEvaluatorClass {defaultValue=/Script/FortniteAIServer.FortAthenaAIBotEvaluator_ThrowableAttack}
   }
  }
  Evaluating...Range Attack {
   class=FortAthenaBTService_AIEvaluator
   properties {
    bCallTickOnSearchStart=true
    interval=0.2
    nodeName=Evaluating...Range Attack
    randomDeviation=0.05
    aIEvaluatorClass {defaultValue=/Game/Athena/AI/Phoebe/Evaluators/Phoebe_Evaluator_RangeAttack.Phoebe_Evaluator_RangeAttack_C}
   }
  }
  Evaluating...AimDownSight {
   class=FortAthenaBTService_AIEvaluator
   properties {
    interval=0.5
    nodeName=Evaluating...AimDownSight
    aIEvaluatorClass {defaultValue=/Script/FortniteAIServer.FortAthenaAIBotEvaluator_AimDownSight}
   }
  }
  Executing...Weapons Actions {class=FortAthenaBTService_ManageWeapon, properties {interval=1, randomDeviation=0.3}}
  Focus on target {
   class=FortAthenaBTService_AIEvaluator
   properties {interval=0.5, nodeName=Focus on target, aIEvaluatorClass {defaultValue=&FortAthenaAIBotEvaluator_HandleFocusing}}
  }
 ]
}
