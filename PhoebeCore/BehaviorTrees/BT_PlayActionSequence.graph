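# Behavior tree that has an AI bot play an action sequence, optionally moving into position first.
# A root Sequence with a SetInTacticalAction service runs a Selector, then the Play Action Sequence task.
# The Selector tries a blackboard-gated MoveTo to the pre-positioning location and falls back to Always Succeed, so disabled or failed positioning never blocks playback.
# The play task takes its sequence inputs from single, multiple and dynamic blackboard keys.
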BehaviorTree /PhoebeCore/BehaviorTrees/BT_PlayActionSequence
blackboard=&/BB_PlayActionSequence
root {
 name=Sequence
 class=BTComposite_Sequence
 children [
  Selector {
   class=BTComposite_Selector
   children [
    "MoveTo : Positioning before Action Sequence" {
     class=FortAthenaBTTask_BotMoveTo
     decorators [Pre-Positioning Enabled ? {
       class=BTDecorator_Blackboard
       properties {
        arithmeticOperation=Equal
        flowAbortMode=None
        intValue=0
        nodeName=Pre-Positioning Enabled ?
        stringValue=ExecutionError
        blackboardKey {selectedKeyName=PlayActionSequence_PrePositioningEnabled}
       }
      }]
     properties {
      movementResultKeyName=None
      nodeName=MoveTo : Positioning before Action Sequence
      acceptableRadius {defaultValue=10}
      allowRandomWobble {defaultValue=true}
      bAllowStrafe {defaultValue=true}
      bTrackMovingGoal {defaultValue=false}
      blackboardKey {selectedKeyName=PlayActionSequence_MoveToLocation}
      observedBlackboardValueTolerance {defaultValue=4.75}
     }
    }
    Always Succeed {class=FortBTTask_AlwaysSucceed, children [], decorators [], properties {bIgnoreRestartSelf=false, nodeName=Always Succeed}, services []}
   ]
   decorators []
   properties {
    children [
     {
      childComposite=null
      childTask=&self:FortAthenaBTTask_BotMoveTo_3
      decorators [&self:BTDecorator_Blackboard_0]
     }
     {childComposite=null, childTask=&self:FortBTTask_AlwaysSucceed_1, decorators []}
    ]
    services []
   }
   services []
  }
  Play Action Sequence {
   class=FortAthenaBTTask_PlayActionSequence
   children []
   decorators []
   properties {
    bIgnoreRestartSelf=false
    dynamicActionSequenceKeyName=PlayActionSequence_Dynamic
    nodeName=Play Action Sequence
    actionSequence {baseClass=/Script/FortniteAIServer.AIActionSequencer_Sequence, defaultValue=null, key=PlayActionSequence_Sequence}
    actionSequences {
     bCanEditDefaultValueType=false
     key=PlayActionSequence_Sequences
     defaultValue {_structType=/Script/DataRegistry.SoftDataRegistryOrTable, bUseDataRegistry=false, registryType=None, table=null}
    }
   }
   services []
  }
 ]
 decorators []
 properties {
  children [
   {childComposite=&self:BTComposite_Selector_1, childTask=null}
   {childTask=&self:FortAthenaBTTask_PlayActionSequence_0}
  ]
  services [&self:FortAthenaBTService_SetInTacticalAction_1]
 }
 services [FortAthenaBTService_SetInTacticalAction_1 {
   class=FortAthenaBTService_SetInTacticalAction
   properties {shouldSetInTacticalAction {key=None}, tacticalActionIdentifierTag {defaultValue {_structType=&GameplayTag}}}
  }]
}
